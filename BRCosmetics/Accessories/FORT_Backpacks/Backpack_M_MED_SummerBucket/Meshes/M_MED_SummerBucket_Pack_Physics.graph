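# Physics asset for the SummerBucket medium male backpack cosmetic, giving its dangling basket and hat bones physics bodies.
# Single capsules cover the pack, basket and hat bones, with the basket and hat each constrained directly to the pack.
# Angular limits let the basket swing mainly about one axis and the hat wobble only slightly, with zero twist range on both.
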PhysicsAsset /BRCosmetics/Accessories/FORT_Backpacks/Backpack_M_MED_SummerBucket/Meshes/M_MED_SummerBucket_Pack_Physics
bodies {
 dyn_basket_01 [{
   shapeType=Capsule
   length=9.5993
   radius=2.6439
   shapeName=dyn_basket_01_capsule
   center=<11.643 0.1263 -0.152>
   rotation=<rot -0.0009 -179.3206 -94.7102>
  }]
 dyn_hat_01 [{
   shapeType=Capsule
   length=0.7699
   radius=4.5529
   shapeName=dyn_hat_01_capsule
   center=<3.3581 0.1888 0.0966>
   rotation=<rot 85.4917 99.7372 -80.0001>
  }]
 pack [{
   shapeType=Capsule
   length=16.7614
   radius=9.1619
   shapeName=pack_capsule
   center=<0.8077 -9.9134 1.8332>
   rotation=<rot -19.6211 -86.3043 -168.3645>
  }]
}
constraints [
 {bone1Name=dyn_basket_01, bone2Name=pack, swing1LimitDegrees=0, swing2LimitDegrees=30, twistLimitDegrees=0}
 {bone1Name=dyn_hat_01, bone2Name=pack, swing1LimitDegrees=2, swing2LimitDegrees=4, twistLimitDegrees=0}
]
